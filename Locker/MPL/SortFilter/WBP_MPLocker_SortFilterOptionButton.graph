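# Sort/filter option button widget for the Locker, wrapping a UIKit toggle button styled as a left-aligned checkbox without all-caps text.
# Pre-construct applies the toggle type, text case and vertical text centering.
# Clicking broadcasts OnButtonClicked with its FortOptionItemVM, then fires an element-interacted event named by the option's label.
# BuildButtonText appends a count to the label unless it is negative; SetButtonText applies text to the button.
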WidgetBlueprint /Locker/MPL/SortFilter/WBP_MPLocker_SortFilterOptionButton
graphs [
 BuildButtonText {nodes [
   n1 = BuildButtonText!() -> (then, Text:Text, Count:int) {is=title K2Node_FunctionEntry}
   > Return Node!(
    execute
    Button Text:Text=Select(
     Option 0:Text=Format Text!(Format="LOCTEXT(\"{Text} ({Count})\")", Text:Text=@n1, Count:int=@n1.Count) -> (Result)
     Option 1:Text=@n1
     Index="integer < integer"(A=@n1.Count)
    ) -> (ReturnValue:Text)
   )
  ]}
 EventGraph {nodes [
   "On Clicked (WBP_UIKit_ButtonWithToggle)"!() -> (OutputDelegate, then, Button:Common Button Base Object Reference) {is=title K2Node_ComponentBoundEvent}
   > Call On Button Clicked!(
    execute:Exec
    self:WBP MPLocker Sort Filter Option Button Object Reference
    Button:Fort Option Item VM Object Reference=Get FortOptionItemVM() -> (FortOptionItemVM:Fort Option Item VM Object Reference)
   ) -> (then:Exec) {is=title K2Node_CallDelegate}
   n1 = Get FortOptionItemVM()
   > FireEventElementInteractedByName!(
    execute
    Element Name:String=Get Source String from Text!(Text:Text=Get Label!(self:Fort Sort Filter Item VM Object Reference=@n1) -> (Label:Text)) -> (ReturnValue:String) {is=title K2Node_CallFunction}
    WidgetCallingFrom:Widget Object Reference=Self-Reference()
   ) -> (then, Success:bool)
   Event PreConstruct()
   n2 = Get WBP_UIKit_ButtonWithToggle() -> (WBP_UIKit_ButtonWithToggle:WBP UIKit Button With Toggle Object Reference)
   > SetToggleType!(execute, self:WBP UIKit Button With Toggle Object Reference=@n2, Type:E_UI_ToggleType Enum=NewEnumerator2) -> (then)
   > SetTextCase!(
    execute
    self:Common Text Object Reference=Get TextBlock!(self:WBP UIKit Block Text Object Reference=Get Block_Text!(self:WBP UIKit Button Generic Base Object Reference=@n2) -> (Block_Text:WBP UIKit Block Text Object Reference)) -> (TextBlock:Common Text Object Reference)
    bUseAllCaps:bool=false
   ) -> (then)
   n3 = Get WBP_UIKit_ButtonWithToggle()
   > n4 = Cast To GridSlot!(execute, Object=Get Slot!(self:Widget Object Reference=Get Block_Text(self=@n3)) -> (Slot:Panel Slot Object Reference)) -> (then, CastFailed, AsGrid Slot:Grid Slot Object Reference) {is=title K2Node_DynamicCast}
   > SetVerticalAlignment(self:Grid Slot Object Reference=@n4, InVerticalAlignment=VAlign_Center)
  ]}
 OnButtonClicked {nodes [OnButtonClicked!() -> (then, Button:Fort Option Item VM Object Reference) {is=title K2Node_FunctionEntry}]}
 SetButtonText {nodes [
   n1 = SetButtonText!() -> (then, Text:Text) {is=title K2Node_FunctionEntry}
   > SetText!(execute, self:WBP UIKit Button Generic Base Object Reference=Get WBP_UIKit_ButtonWithToggle(), Text:Text=@n1) -> (then)
  ]}
]
variables [OnButtonClicked {}]
widgets {
 info {
  inheritedWidgetCount=0
  parentClass=&UserWidget
  rootWidgetClass=/FortUILibrary/UIKit/Buttons/WBP_UIKit_ButtonWithToggle.WBP_UIKit_ButtonWithToggle_C
  widgetCount=1
 }
 widgets [WBP_UIKit_ButtonWithToggle {
   widgetClassPath=/FortUILibrary/UIKit/Buttons/WBP_UIKit_ButtonWithToggle
   bIsVariable=true
   widget=WBP_UIKit_ButtonWithToggle
   properties {
    iconHorizontalAlignment=HAlign_Left
    textJustification=Left
    textSecondJustification=Left
    textTransformPolicy=None
    toggleType=CheckBox
    font {
     bForceMonospaced=false
     bMaterialIsStencil=false
     fontMaterial=/FortUILibrary/UIKit/Buttons/MaterialInstances/GenericButtons/MI_UIKit_ButtonWithToggle_Text
     fontObject=&HeadingNow
     letterSpacing=48
     monospacedWidth=1
     size=18
     skewAmount=0
     typefaceFontName=75 Medium
     outlineSettings {
      bApplyOutlineToDropShadows=false
      bMiteredCorners=false
      bSeparateFillAlpha=false
      outlineMaterial=null
      outlineSize=0
      outlineColor=<rgba 0 0 0 1>
     }
    }
   }
  }]
}
