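# Widget for one labeled section of the MP Locker sort/filter options, showing option buttons in a wrapping Dynamic Entry Box.
# SetupItems resets the pooled entries and unbinds them, then creates a button per option viewmodel, with toggle state taken from bIsActive.
# Button clicks are re-broadcast through OnItemClicked.
# PreConstruct sets the section label, using placeholder text at design time when empty, and adjusts the entry box's top padding on mobile.
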WidgetBlueprint /Locker/MPL/SortFilter/WBP_MPLocker_SortFilterOptions_Section
graphs [
 EventGraph {nodes [
   n1 = Event PreConstruct()
   n2 = Get SectionLabel() -> (SectionLabel:Text)
   n3 = TextIsEmpty(InText=@n2)
   > "SetText (Text)"(
    self=Get Text_SectionName() -> (Text_SectionName:Common Text Object Reference)
    InText=Select(Option 0:Text=@n2, Option 1:Text="INVTEXT(\"[PH] Section Text\")", Index=AND Boolean(A=@n1.IsDesignTime, B=@n3)) -> (ReturnValue:Text)
   )
   > SetCompactUIStyle() {is=title K2Node_CallFunction}
  ]}
 HandleButtonClicked {nodes [
   n1 = HandleButtonClicked!() -> (then, Button:Fort Option Item VM Object Reference) {is=title K2Node_FunctionEntry}
   > Call On Item Clicked!(execute:Exec, self:WBP MPLocker Sort Filter Options Section Object Reference, ItemFilterVM:Fort Sort Filter Item VM Object Reference=@n1) -> (then:Exec) {is=title K2Node_CallDelegate}
  ]}
 OnItemClicked {nodes [OnItemClicked!() -> (then, ItemFilterVM:Fort Sort Filter Item VM Object Reference) {is=title K2Node_FunctionEntry}]}
 ResetEntries {nodes [
   ResetEntries() {is=title K2Node_FunctionEntry}
   > n1 = For Each Loop(Array:Array of User Widget Object References=GetAllEntries!(self:Dynamic Entry Box Base Object Reference=Get DynBox_Entries() -> (DynBox_Entries:Dynamic Entry Box Object Reference)) -> (ReturnValue:Array of User Widget Object References)) -> (Array Element:User Widget Object Reference)
   Reset!(execute=@n1.Completed, self:Dynamic Entry Box Object Reference=Get DynBox_Entries(), bDeleteWidgets:bool=false) -> (then)
   n2 = Cast To WBP_MPLocker_SortFilterOptionButton!(execute=@n1.LoopBody, Object=@n1) -> (then, CastFailed, AsWBP MPLocker Sort Filter Option Button:WBP MPLocker Sort Filter Option Button Object Reference) {is=title K2Node_DynamicCast}
   > Unbind Event from On Button Clicked(self:WBP MPLocker Sort Filter Option Button Object Reference=@n2, Delegate=Create Event()) {is=title K2Node_RemoveDelegate}
  ]}
 SetCompactUIStyle {nodes [
   SetCompactUIStyle() {is=title K2Node_FunctionEntry}
   > Branch(Condition=IsMobileGame())
   n1 = SlotAsVerticalBoxSlot(Widget=Get DynBox_Entries())
   n2 = Get!(self:Vertical Box Slot Object Reference=@n1) -> (Padding_Left:float=~, Padding_Top:float=~, Padding_Right:float=~, Padding_Bottom:float=~)
   > SetPadding!(
    execute
    self:Vertical Box Slot Object Reference=@n1
    InPadding_Left:float=@n2
    InPadding_Top:float=24
    InPadding_Right:float=@n2.Padding_Right
    InPadding_Bottom:float=@n2.Padding_Bottom
   ) -> (then)
  ]}
 SetupItems {nodes [
   n1 = SetupItems!() -> (then, Array:Array of Fort Option Item VM Object References) {is=title K2Node_FunctionEntry}
   > ResetEntries() {is=title K2Node_CallFunction}
   > n2 = For Each Loop(Array:Array of Fort Option Item VM Object References=@n1) -> (Array Element:Fort Option Item VM Object Reference)
   > n3 = Create Entry!(execute, self:Dynamic Entry Box Object Reference=Get DynBox_Entries()) -> (then, ReturnValue:User Widget Object Reference)
   > n4 = Cast To WBP_MPLocker_SortFilterOptionButton(Object=@n3)
   > SetFortOptionItemVM!(execute, self:WBP MPLocker Sort Filter Option Button Object Reference=@n4, Viewmodel:Fort Option Item VM Object Reference=@n2) -> (then)
   > SetIsSelected!(
    execute
    self:Common Button Base Object Reference=Get WBP_UIKit_ButtonWithToggle!(self:WBP MPLocker Sort Filter Option Button Object Reference=@n4) -> (WBP_UIKit_ButtonWithToggle:WBP UIKit Button With Toggle Object Reference)
    InSelected:bool=Get bIsActive!(self:Fort Sort Filter Item VM Object Reference=@n2) -> (bIsActive:bool)
    bGiveClickFeedback:bool=false
   ) -> (then)
   > Bind Event to On Button Clicked(self:WBP MPLocker Sort Filter Option Button Object Reference=@n4, Delegate=Create Event()) {is=title K2Node_AddDelegate}
  ]}
]
variables [
 Button {type=CommonButtonBase, default=null}
 OnItemClicked {}
 SectionLabel {type=string, default=""}
 SectionString {type=string, default=""}
]
widgets {
 info {inheritedWidgetCount=0, parentClass=&UserWidget, rootWidgetClass=&VerticalBox, widgetCount=3}
 widgets [
  VerticalBox_0 {widgetClassPath=&VerticalBox, widget=VerticalBox_0}
  Text_SectionName {
   widgetClassPath=&CommonTextBlock
   bIsVariable=true
   parent=VerticalBox_0
   widget=Text_SectionName
   properties {
    bAutoCollapseWithEmptyText=true
    clipping=OnDemand
    scrollStyle=/FortUILibrary/Styles/Typography/Foundation/TextScrollStyle-NoFade.TextScrollStyle-NoFade_C
    text=Section
    textOverflowPolicy=Ellipsis
    font {
     bForceMonospaced=false
     bMaterialIsStencil=false
     fontMaterial=null
     fontObject=&HeadingNow
     letterSpacing=0
     monospacedWidth=1
     size=20.25
     skewAmount=0
     typefaceFontName=75 Medium
     outlineSettings {
      bApplyOutlineToDropShadows=false
      bMiteredCorners=false
      bSeparateFillAlpha=false
      outlineMaterial=null
      outlineSize=0
      outlineColor=<rgba 0 0 0 1>
     }
    }
    shadowOffset=<0 0>
   }
   slot {class=VerticalBoxSlot, properties {padding=<ltrb 10 0 10 0>}}
  }
  DynBox_Entries {
   widgetClassPath=&DynamicEntryBox
   bIsVariable=true
   parent=VerticalBox_0
   widget=DynBox_Entries
   properties {
    entryBoxType=Wrap
    entryWidgetClass=&/WBP_MPLocker_SortFilterOptionButton.WBP_MPLocker_SortFilterOptionButton_C
    numDesignerPreviewEntries=10
    entrySpacing=<12 8>
   }
   slot {class=VerticalBoxSlot, properties {padding=<ltrb 2 10 10 20>}}
  }
 ]
}
